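# MetaSound patch that turns speed, azimuth and elevation movement inputs into smoothed normalized controls for spatial travel audio.
# Each input is range-mapped and interpolated, with elevation taken as absolute before mapping and azimuth folded to absolute after.
# Deferred variable reads hold previous speed and CurrentTime values to derive a scaled, smoothed SpeedDelta.
# Smoothed azimuth drives MSP_OneshotGate, which fires TurnActivated and TurnFinished using threshold, conditions and retrigger delay.
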MetaSoundPatch /FortMetasound/SpatialTravel/MSP_SpatialTravelController
inputs [
 Azimuth.EnterCondition {type=Enum:TriggerComparisonType}
 Azimuth.ExitCondition {type=Enum:TriggerComparisonType}
 Azimuth.MaxRange {type=Float}
 Azimuth.MinRange {type=Float}
 Azimuth.Movement {type=Float}
 Azimuth.RetriggerDelay {type=Time}
 Azimuth.Threshold {type=Float}
 CurrentTime {type=Float}
 Elevation.ChangeMax {type=Float}
 Elevation.ChangeMin {type=Float}
 Elevation.Movement {type=Float}
 InterpSpeed {type=Time}
 Speed.Movement {type=Float}
 Speed.SpeedFloor {type=Float}
 Speed.SpeedMax {type=Float}
 Speed.SpeedMin {type=Float}
]
outputs [
 NormalizedAzimuth {type=Float}
 NormalizedElevation {type=Float}
 NormalizedSpeed {type=Float}
 SpeedDelta {type=Float}
 TurnActivated {type=Trigger}
 TurnFinished {type=Trigger}
]
pages [Default {
  input_defaults {
   Azimuth.EnterCondition=5
   Azimuth.ExitCondition=2
   Azimuth.MaxRange=180
   Azimuth.MinRange=-180
   Azimuth.Movement=0
   Azimuth.RetriggerDelay=1
   Azimuth.Threshold=0.5
   CurrentTime=0
   Elevation.ChangeMax=100
   Elevation.ChangeMin=0
   Elevation.Movement=0
   InterpSpeed=0.5
   Speed.Movement=0
   Speed.SpeedFloor=0
   Speed.SpeedMax=1
   Speed.SpeedMin=0
  }
  nodes [
   InitVariable.Float()
   n1 = MapRange.MapRange:Float(In=^Speed.Movement, Out Range A=^Speed.SpeedFloor, In Range B=^Speed.SpeedMax, In Range A=^Speed.SpeedMin)
   n2 = VariableDeferredAccessor.Float(Variable=InitVariable.Float().Variable)
   n3 = UE.Subtract:Float(
    PrimaryOperand=VariableAccessor.Float(Variable=VariableMutator.Float(Value=@n1.Out Value, Variable=@n2.Variable).Variable).Value
    AdditionalOperands=@n2.Value
   )
   n4 = VariableDeferredAccessor.Float(Variable=InitVariable.Float().Variable)
   n5 = UE.Subtract:Float(
    PrimaryOperand=VariableAccessor.Float(Variable=VariableMutator.Float(Value=^CurrentTime, Variable=@n4.Variable).Variable).Value
    AdditionalOperands=@n4.Value
   )
   InterpTo_0 = UE.InterpTo:Audio(Target=UE.Multiply:Float(PrimaryOperand=UE.Divide:Float(PrimaryOperand=@n3.Out, AdditionalOperands=@n5.Out).Out) {also {defaults {AdditionalOperands=1000}}}.Out) {also {defaults {Interp Time=1}}}
   InterpTo_1 = UE.InterpTo:Audio(
    Interp Time=^InterpSpeed
    Target=MapRange.MapRange:Float(In Range B=^Elevation.ChangeMax, In Range A=^Elevation.ChangeMin, In=AbsoluteValue.Abs:Float(Input=^Elevation.Movement).Value).Out Value
   )
   InterpTo_3 = UE.InterpTo:Audio(Interp Time=^InterpSpeed, Target=@n1.Out Value)
   InterpTo_2 = UE.InterpTo:Audio(
    Interp Time=^InterpSpeed
    Target=AbsoluteValue.Abs:Float(Input=MapRange.MapRange:Float(In Range B=^Azimuth.MaxRange, In Range A=^Azimuth.MinRange, In=^Azimuth.Movement) {also {defaults {Out Range A=-1}}}.Out Value).Value
   )
   MSP_OneshotGate_0 = &MSP_OneshotGate(
    EnterCondition=^Azimuth.EnterCondition
    ExitCondition=^Azimuth.ExitCondition
    RetriggerDelay=^Azimuth.RetriggerDelay
    OneshotThreshold=^Azimuth.Threshold
    Value=@InterpTo_2.Value
   )
  ]
  links [
   {dst="@SpeedDelta", src=InterpTo_0.Value}
   {dst="@NormalizedElevation", src=InterpTo_1.Value}
   {dst="@NormalizedAzimuth", src=InterpTo_2.Value}
   {dst="@NormalizedSpeed", src=InterpTo_3.Value}
   {dst="@TurnActivated", src=MSP_OneshotGate_0.OnGateEnter}
   {dst="@TurnFinished", src=MSP_OneshotGate_0.OnGateExit}
  ]
 }]
