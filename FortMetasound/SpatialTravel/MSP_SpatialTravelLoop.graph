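# MetaSound patch that plays a random wave from an array as a stereo loop layer for spatial travel audio.
# FadeIn picks the wave, starts it at either zero or a random point within its duration, and triggers MSP_FadeToValue to ramp gain.
# FadeOut reverses that fade and stops the player once it completes.
# A pitch control mapped between min and max drives pitch shift, while a volume control scales the fade gain.
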MetaSoundPatch /FortMetasound/SpatialTravel/MSP_SpatialTravelLoop
inputs [
 FadeIn {type=Trigger}
 FadeOut {type=Trigger}
 Loop.FadeTime {type=Float}
 Loop.PitchControl {type=Float}
 Loop.PitchhMin {type=Float}
 Loop.PitchMax {type=Float}
 Loop.RandomStartPos {type=Bool}
 Loop.Sounds {type=WaveAsset:Array}
 Loop.VolumeControl {type=Float}
]
outputs [
 On Loop Finished {type=Trigger}
 On Play {type=Trigger}
 OutL {type=Audio}
 OutR {type=Audio}
]
pages [Default {
  input_defaults {
   FadeIn=null
   FadeOut=null
   Loop.FadeTime=0.5
   Loop.PitchControl=null
   Loop.PitchMax=1
   Loop.PitchhMin=0
   Loop.RandomStartPos="<MetaSoundBuilderResult.SUCCEEDED: 0>"
   Loop.VolumeControl=1
   Loop.Sounds []
  }
  nodes [
   ArrayRandomGet_0 = Array.Random Get:WaveAsset:Array(Next=^FadeIn, In Array=^Loop.Sounds) {also {defaults {Weights []}}}
   n1 = UE.Reroute(Value=@ArrayRandomGet_0.Value)
   GetWaveDuration_0 = "UE.Get Wave Duration: "(Wave=@n1.Value)
   n2 = TriggerCompare.Clamp:Bool(A=^Loop.RandomStartPos, Compare=@ArrayRandomGet_0.On Next)
   n3 = UE.RandomTime(Max=@GetWaveDuration_0.Duration, Next=@n2.False)
   TriggerRouteTriggerRouteTime2_0 = "TriggerRoute.Trigger Route (Time, 2):Time"(Set 1=@n3.On Next, Value 1=@n3.Value, Set 0=@n2.True) {also {defaults {Value 0=0}}}
   n4 = /FortMetasound/Patches/MSP_FadeToValue(ReverseFade=^FadeOut, FadeTime=UE.ConversionFloatToTime(In=^Loop.FadeTime).Out, Fade=@TriggerRouteTriggerRouteTime2_0.On Set)
   n5 = VariableMutator.Float(Variable=InitVariable.Float().Variable, Value=@n4.Out Value)
   n6 = UE.Reroute(Value=@n4.OnDoneReverseFade)
   n7 = InitVariable.Float()
   n8 = MapRange.MapRange:Float(In=^Loop.PitchControl, Out Range B=^Loop.PitchMax, Out Range A=^Loop.PitchhMin)
   WavePlayer_0 = UE.Wave Player:Stereo(
    Wave Asset=UE.Reroute(Value=UE.Reroute(Value=@n1.Value).Value).Value
    Start Time=UE.Reroute(Value=UE.Reroute(Value=@TriggerRouteTriggerRouteTime2_0.Value).Value).Value
    Stop=@n6.Value
    Play=@TriggerRouteTriggerRouteTime2_0.On Set
    Pitch Shift=VariableAccessor.Float(Variable=VariableMutator.Float(Variable=@n7.Variable, Value=@n8.Out Value).Variable).Value
   ) {also {defaults {Loop="<MetaSoundBuilderResult.SUCCEEDED: 0>"}}}
   AudioMixerAudioMixerStereo2_0 = "AudioMixer.Audio Mixer (Stereo, 2)"(
    Gain 0=UE.Multiply:Float(AdditionalOperands=^Loop.VolumeControl, PrimaryOperand=VariableAccessor.Float(Variable=@n5.Variable).Value).Out
    In 0 L=@WavePlayer_0.Out Left
    In 0 R=@WavePlayer_0.Out Right
   )
   InitVariable.Bool()
   InitVariable.Float()
   VariableMutator.Trigger(Variable=InitVariable.Trigger().Variable, Value=@n6.Value)
  ]
  links [
   {dst="@OutL", src=AudioMixerAudioMixerStereo2_0.Out L}
   {dst="@OutR", src=AudioMixerAudioMixerStereo2_0.Out R}
   {dst="@On Loop Finished", src=WavePlayer_0.On Finished}
   {dst="@On Play", src=WavePlayer_0.On Play}
  ]
 }]
